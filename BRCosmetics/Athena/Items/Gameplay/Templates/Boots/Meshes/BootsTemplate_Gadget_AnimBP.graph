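# Animation Blueprint for the boots template gadget mesh that makes it follow the pose of its attached parent mesh.
# The base pose is copied from the attached parent, while a second pose comes from Retarget Pose From Mesh using the boots gadget IK retargeter.
# A per-bone layered blend applies the retargeted pose from the ball_l and ball_r bones down.
# The event graph is empty, and the two ball rotator variables are not read.
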AnimBlueprint /BRCosmetics/Athena/Items/Gameplay/Templates/Boots/Meshes/BootsTemplate_Gadget_AnimBP
target_skeleton=&/BootsTemplate_Gadget_Skeleton
graphs [
 AnimGraph {nodes [Output Pose(Result=Layered blend per bone(
     BasePose=Copy Pose From Mesh() {bUseAttachedParent=true}
     BlendPoses_0=Retarget Pose From Mesh!() -> (Pose:Pose Link Structure) {
      is=title AnimGraphNode_RetargetPoseFromMesh
      iKRetargeterAsset=&/RTG_BootsTemplate_Gadget
     }
    ) {layerSetup [{branchFilters [{boneName=ball_l}, {boneName=ball_r}]}]})]}
 EventGraph {links [], nodes []}
]
variables [
 ball_l rot {type=Rotator}
 ball_r rot {type=Rotator}
]
